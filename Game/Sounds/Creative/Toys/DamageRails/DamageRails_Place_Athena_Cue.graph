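# Sound cue for placing the Damage Rails creative toy in Athena, built from existing trap placement sounds.
# A mixer layers a random pick of two electric wall trap placement waves with a quieter enemy launcher placement wave, which passes through a modulator.
# It uses the Athena trap-activation attenuation and a concurrency rule set that, by name, caps instances at three per owner and blocks new ones.
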SoundCue /Game/Sounds/Creative/Toys/DamageRails/DamageRails_Place_Athena_Cue
root=Mixer_1
nodes [
 {id=Mixer_1, children [Random_19,Modulator_17], properties {inputVolume [1,0.7]}}
 {id=Modulator_17, children [WavePlayer_142]}
 {id=Random_19, children [WavePlayer_69, WavePlayer_70], properties {weights [1,1]}}
 {id=WavePlayer_142, properties {soundWaveAssetPtr=/Game/Sounds/Fort_Traps/Enemy_Launcher/EnemyLauncher_Placed_001}}
 {id=WavePlayer_69, properties {soundWaveAssetPtr=/Game/Sounds/Fort_Traps/Wall_Electric/WallTrap_Electric_Placed_001}}
 {id=WavePlayer_70, properties {soundWaveAssetPtr=/Game/Sounds/Fort_Traps/Wall_Electric/WallTrap_Electric_Placed_002}}
]
properties {
 attenuationSettings=/Game/Sounds/Attenuation/Fort_Traps_Activate_Athena_Att
 maxDistance=3500
 soundClassObject=&Fort_WeaponMelee_PSM
 volumeMultiplier=0.55
 attenuationOverrides {falloffDistance=2400, attenuationShapeExtents {x=600}}
 concurrencyOverrides {resolutionRule=PreventNew}
 concurrencySet [/Game/Sounds/Concurrency_Rules/ThreeMaximum_Owner_PreventNew]
}
