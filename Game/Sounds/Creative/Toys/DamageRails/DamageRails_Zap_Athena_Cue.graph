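# Sound cue for the zap of the Creative damage rail toy in Athena, classed as a Fort trap sound.
# A Mixer layers a randomly chosen, modulated electric wall trap firing sound with a fixed Cube angry-zap wave, weighting the Cube layer louder.
# It uses trap-activate attenuation and overrides concurrency to block new instances at the limit rather than cut playing ones.
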SoundCue /Game/Sounds/Creative/Toys/DamageRails/DamageRails_Zap_Athena_Cue
root=Mixer_19
nodes [
 {id=Mixer_19, children [Modulator_9,WavePlayer_1], properties {inputVolume [0.8,1.6]}}
 {id=Modulator_9, children [Random_18]}
 {id=Random_18, children [WavePlayer_65, WavePlayer_66, WavePlayer_68], properties {weights [1,1,1]}}
 {id=WavePlayer_1, properties {soundWaveAssetPtr=/Game/Athena/Sounds/Events/Cube/Cube_AngryZap_01}}
 {id=WavePlayer_65, properties {soundWaveAssetPtr=/Game/Sounds/Fort_Traps/Wall_Electric/WallTrap_Electric_Fire_004}}
 {id=WavePlayer_66, properties {soundWaveAssetPtr=/Game/Sounds/Fort_Traps/Wall_Electric/WallTrap_Electric_Fire_001}}
 {id=WavePlayer_68, properties {soundWaveAssetPtr=/Game/Sounds/Fort_Traps/Wall_Electric/WallTrap_Electric_Fire_003}}
]
properties {
 attenuationSettings=/Game/Sounds/Attenuation/Fort_Traps_Activate_Athena_Att
 bOverrideConcurrency=true
 maxDistance=3500
 soundClassObject=&Fort_Traps
 volumeMultiplier=0.6
 attenuationOverrides {falloffDistance=3200, attenuationShapeExtents {x=800}}
 concurrencyOverrides {resolutionRule=PreventNew}
}
